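; Legacy glider Blueprint for the BlackMondayCake cosmetic that drives parameters on its trail particle effect.
; When the chute becomes visible, it sets "SepTrans" from the r.SeparateTranslucency console variable.
; Trail updates store RightDot and, once the animation reports full deployment, call AirFoilCheck.
; AirFoilCheck weights left and right trail alpha oppositely by RightDot sign, balancing them inside a small dead zone, behind a DoOnce gate.
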(event EventOnChuteVisibilityChanged
  (params (bParachuteVisibility :type Boolean))
  (|CallParentFunction bParachuteVisibility)
  (if bParachuteVisibility
    (bind _returnvalue (Development|GetConsoleVariableBoolValue "r.SeparateTranslucency"))
    (Effects|Components|ParticleSystem|SetFloatParameter (CallFunction|GetTrailVFXComponent) "SepTrans" (select _returnvalue 0.0 1.0))))

(event Parachute|EventOnParachuteTrailUpdated
  (params (MovementLengthSquared :type Float%20%28single-precision%29) (ForwardDot :type Float%20%28single-precision%29) (RightDot :type Float%20%28single-precision%29) (RotationalMovementAcceleration :type Float%20%28single-precision%29))
  (|CallParentFunction MovementLengthSquared ForwardDot RightDot RotationalMovementAcceleration)
  (Variables|Default|SetRightDOT RightDot)
  (if (CallFunction|IsGliderFullyDeployedAccordingtoAnim)
    (CallFunction|AirFoilCheck (Variables|Default|GetRightDOT))))

(fn AirFoilCheck
  (params (Right%20Dot :type Float%20%28double-precision%29))
  (bind _returnvalue (CallFunction|GetTrailVFXComponent))
  (bind _returnvalue_2 (CallFunction|GetTrailVFXComponent))
  (bind _returnvalue_3 (CallFunction|GetTrailVFXComponent))
  (if (not (Math|Float|NearlyEqual(Float) 0.0 (Variables|Default|GetRightDOT)))
    (Utilities|FlowControl|Sequence
      (:then_0
        _)
      (:then_1
        _)))
  (Utilities|FlowControl|DoOnce false
    (:Completed
      (if (>= (Variables|GetRight%20Dot) 0.1)
        (Effects|Components|ParticleSystem|SetFloatParameter _returnvalue "TrailAlphaLeft" 2.0)
        (Effects|Components|ParticleSystem|SetFloatParameter _returnvalue "TrailAlphaRight" 0.5)
        (elif (<= (Variables|GetRight%20Dot) -0.1)
          (Effects|Components|ParticleSystem|SetFloatParameter _returnvalue_2 "TrailAlphaLeft" 0.5)
          (Effects|Components|ParticleSystem|SetFloatParameter _returnvalue_2 "TrailAlphaRight" 2.0)
          (else
            (Effects|Components|ParticleSystem|SetFloatParameter _returnvalue_3 "TrailAlphaLeft" 1.0)
            (Effects|Components|ParticleSystem|SetFloatParameter _returnvalue_3 "TrailAlphaRight" 1.0)))))))

(fn ConstructionScript
  (|CallParentFunction))
