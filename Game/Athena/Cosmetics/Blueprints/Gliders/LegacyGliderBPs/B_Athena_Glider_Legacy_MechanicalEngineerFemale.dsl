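; Legacy glider cosmetic Blueprint for the Mechanical Engineer Female glider that controls its spark effect component.
; When the glider fully deploys, it calls the parent and activates the sparks.
; On chute visibility changes, it calls the parent, deactivates the sparks when hidden, and activates them when visible only in a front-end preview whose animation reports full deployment.
; The construction script only calls the parent.
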(event EventOnChuteVisibilityChanged
  (params (bParachuteVisibility :type Boolean))
  (|CallParentFunction bParachuteVisibility)
  (if bParachuteVisibility
    (if (and (Variables|FortPlayerParachute|GetbIsFrontEndPreview) (CallFunction|IsGliderFullyDeployedAccordingtoAnim))
      (Components|Activation|Activate (Variables|Default|GetP_Glider_Mechanical_Engineer_Spark) false))
    (else
      (Components|Activation|Deactivate (Variables|Default|GetP_Glider_Mechanical_Engineer_Spark)))))

(event EventOnGliderFullyDeployed
  (|CallParentFunction)
  (Components|Activation|Activate (Variables|Default|GetP_Glider_Mechanical_Engineer_Spark) false))

(fn ConstructionScript
  (|CallParentFunction))
